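; Tomato Repulsors dual-hand ranged weapon with target lock-on, homing projectiles and an overheat warning sound.
; The server sets LockOnTarget; its OnRep drives sounds and, for the local player, a lock-on widget actor and looping GameplayCue, filtered by tags.
; On projectile spawn, the authority configures homing toward that target.
; Hand-fired events record the last hand and play force feedback, with a temporary Gate workaround for firing while blocked.
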(event Custom|BindOnHandFired
  ;[1 1 1] On hand fired (left - right)
  (RangedWeaponDual|BindEventtoOnHandFired (delegate Custom|HandFired)))

(event EventBeginPlay
  (CallFunction|BindonHandFired))

(event Custom|HandFired
  (params (Hand :type EDualWeaponHand%20Enum :default "LEFT") (bPersistantFire :type Boolean))
  ;[1 1 1] On hand fired (left - right)
  (Variables|Default|SetLastHandFired Hand :!label _label)
  ;;[1 1 1] Comment
  (FortWeapon|PlayForceFeedback false)
  ;;[1 0 0.04289] TEMP!!! this is to solve  handfired being called even if shoot ability is blocked
  (Utilities|FlowControl|Gate true
    (:Exit
      (goto _label))))

(event EventOnPlayWeaponFireFX
  (params (bPersistentFire :type Boolean) (bSecondaryFire :type Boolean))
  (Utilities|FlowControl|Sequence
    (:then_0
      ;[1 1 1] On hand fired (left - right)
      ;;[1 0 0.04289] TEMP!!! this is to solve  handfired being called even if shoot ability is blocked
      (Utilities|FlowControl|Gate true :!label _label
        (:Exit
          (Variables|Default|SetLastHandFired (stale Custom|HandFired Hand))
          ;;[1 1 1] Comment
          (FortWeapon|PlayForceFeedback false)
          (goto _label Close))))
    (:then_1
      (FortWeapon|PlayForceFeedback false bPersistentFire))))

(event EventOnStopWeaponFireFX
  (FortWeapon|StopForceFeedback))

(event FortWeapon|EventOnEquippedAndAttached
  ;[1 1 1] added bool
  (Variables|Default|SetIsTomato true)
  (if (and (Pawn|IsLocallyControlled (Variables|Actor|GetInstigator)) (Variables|Properties|GetShouldSpawnLockOnActor))
    (bind _returnvalue (Game|SpawnActorfromClass "/HighTower/Items/Tomato/Repulsors/Hightower_Tomato_RepulsorsLockOnWidgetActor.Hightower_Tomato_RepulsorsLockOnWidgetActor_C" (struct :Location "0, 0, 0" :Rotation "0, 0, 0" :Scale "1.000000,1.000000,1.000000") "Undefined" "OverrideRootScale" 0 (Variables|Actor|GetInstigator)))
    (Variables|Default|SetLockOnActor _returnvalue)))

(event Custom|NewTarget
  (params (Lock%20On%20Target :type Actor%20Object%20Reference))
  (bind _old_lockontarget (Variables|Default|GetOld_LockOnTarget)
    (:then
      (Projectile|RemoveGameplayCue%28Looping%29Local _old_lockontarget (Variables|Default|GetTag_LockOnFX) (Ability|GameplayCue|MakeGameplayCueParameters 0.0)))
    (:else
      _))
  (if (Pawn|IsLocallyControlled (Variables|Actor|GetInstigator))
    (bind _returnvalue (GameplayTags|HasAnyMatchingGameplayTags(Message) Lock%20On%20Target (GameplayTags|MakeGameplayTagContainerfromArray (Variables|Default|GetNot%20Valid%20Lock%20On%20Tags))))
    (if (not _returnvalue)
      (Utilities|IsValid Lock%20On%20Target
        (:Is%20Valid
          (bind _lockonactor (Variables|Default|GetLockOnActor)
            (:then
              (Class|HightowerTomatoRepulsorsLockonWidgetActor|NewTarget _lockonactor Lock%20On%20Target)
              (Class|HightowerTomatoRepulsorsLockonWidgetActor|LockonStart _lockonactor))
            (:else
              _))
          (Projectile|AddGameplayCue%28Looping%29Local Lock%20On%20Target (Variables|Default|GetTag_LockOnFX) (Ability|GameplayCue|MakeGameplayCueParameters :NormalizedMagnitude 0.0 :RawMagnitude 0.0 :Location "0,0,0" :Normal "0,0,0" :Instigator (Variables|Actor|GetInstigator) :GameplayEffectLevel 1 :AbilityLevel 1 :bReplicateLocationWhenUsingMinimalRepProxy false))
          (Variables|Default|SetOld_LockOnTarget Lock%20On%20Target))
        (:Is%20Not%20Valid
          _)))
    (bind _lockonactor_2 (Variables|Default|GetLockOnActor)
      (:then
        (Class|HightowerTomatoRepulsorsLockonWidgetActor|TargetLost _lockonactor_2)
        (Class|HightowerTomatoRepulsorsLockonWidgetActor|LockonEnd _lockonactor_2))
      (:else
        _))
    (Variables|Default|SetOld_LockOnTarget 0)))

(event EventOnSetTargeting
  (params (bNewIsTargeting :type Boolean)))

(event FortWeapon|EventOnLocalUnEquip
  ;[1 1 1] added bool
  (Variables|Default|SetIsTomato false)
  (bind _lockonactor (Variables|Default|GetLockOnActor)
    (:then
      (Actor|DestroyActor _lockonactor))
    (:else)))

(event FortWeapon|EventOnSpawnProjectile
  (params (SpawnProjectile :type Fort%20Projectile%20Base%20Object%20Reference))
  (bind _turnspeedmax (Variables|Default|GetTurnSpeedMax))
  (Networking|SwitchHasAuthority
    (:Authority
      (Utilities|IsValid (Variables|Default|GetLockOnTarget)
        (:Is%20Valid
          (bind _asb_prj_high_tower_tomato_repulsor (Utilities|Casting|CastToB_Prj_HighTower_Tomato_Repulsor SpawnProjectile))
          (Projectile|SetHomingTurnSpeedMin _asb_prj_high_tower_tomato_repulsor _asb_prj_high_tower_tomato_repulsor (- _turnspeedmax 1.0) :!class /Script/FortniteGame.FortProjectileMovementInterface)
          (Projectile|SetHomingTurnSpeedMax _asb_prj_high_tower_tomato_repulsor _asb_prj_high_tower_tomato_repulsor _turnspeedmax :!class /Script/FortniteGame.FortProjectileMovementInterface)
          (Projectile|SetHomingLockTargetDistanceThreshold _asb_prj_high_tower_tomato_repulsor _asb_prj_high_tower_tomato_repulsor (Variables|Default|GetHomingDistanceThreshold) :!class /Script/FortniteGame.FortProjectileMovementInterface)
          (Projectile|SetHomingTimeUntilMaxTurnSpeed _asb_prj_high_tower_tomato_repulsor _asb_prj_high_tower_tomato_repulsor (Variables|Default|GetTimeUntilMax) :!class /Script/FortniteGame.FortProjectileMovementInterface)
          (Projectile|SetHomingStyle _asb_prj_high_tower_tomato_repulsor _asb_prj_high_tower_tomato_repulsor "LockOn" true :!class /Script/FortniteGame.FortProjectileMovementInterface)
          (Projectile|SetHomingTarget _asb_prj_high_tower_tomato_repulsor _asb_prj_high_tower_tomato_repulsor (Variables|Default|GetLockOnTarget) 0 :!class /Script/FortniteGame.FortProjectileMovementInterface))))))

(event EventPlayWeaponFireCustomAudio
  (params (bPersistentFire :type Boolean) (bSecondaryFire :type Boolean))
  (if (< (Default|GetShotsRemaining) (Variables|Default|GetOverheat%20Warning%20Threshold))
    (bind _returnvalue (Audio|SpawnSoundAttached (Variables|Default|GetOverheat%20Warning%20Sound) (Variables|Visual|GetWeaponMesh)))
    (Audio|Parameter|SetIntegerParameter _returnvalue "ShotsLeft" (Default|GetShotsRemaining) :!class /Script/AudioExtensions.AudioParameterControllerInterface)))

(event Custom|ServerSetLockOnTarget
  (params (LockOnTarget :type Actor%20Object%20Reference))
  (Variables|Default|SetLockOnTarget LockOnTarget))

(fn GetShotsRemaining
  (return (- (Math|Float|Truncate (/ (- 1.0 (/ (Overheat|GetCurrentOverheatValue) (Overheat|GetOverheatingMaxValue))) (/ (Overheat|GetOverheatHeatingValue) 1e+02))) 1)))

(fn OnRep_LockOnTarget
  (if (!= (Variables|Default|GetLockOnTarget) (Variables|Default|GetOld_LockOnTarget))
    (CallFunction|NewTarget (Variables|Default|GetLockOnTarget))
    (Default|CallTomatoLockonUpdated (Variables|Default|GetLockOnTarget))
    (Networking|CanExecuteCosmeticEvents
      (:True
        (if (Pawn|IsLocallyControlled (Variables|Actor|GetInstigator))
          (Variables|Default|GetLockOnTarget
            (:then
              (Audio|PlaySound2D "/HighTower/Audio/Tomato/Hightower_Tomato_Primary_Switch_Target_Cue.Hightower_Tomato_Primary_Switch_Target_Cue"))
            (:else
              (Audio|PlaySound2D "/HighTower/Audio/Tomato/Hightower_Tomato_Primary_Lockon_Cue.Hightower_Tomato_Primary_Lockon_Cue"))))))))

(fn ConstructionScript)
