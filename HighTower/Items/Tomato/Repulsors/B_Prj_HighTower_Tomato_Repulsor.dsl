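; Projectile Blueprint for the HighTower Tomato Repulsors that flies with optional homing and explodes once on impact or lifetime timeout.
; Authority sets the impact location, whose OnRep drives cosmetic effects: player-attached Niagara at the closest bone, or world burst with crack decal.
; Remote BeginPlay feeds the firing hand's socket and timing into tracer material WPO.
; Homing style resets near or past the target; water entry plays a splash.
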(event Custom|Explosion%20VFX
  (bind (_bblockinghit _binitialoverlap _time _distance _location _impactpoint _normal _impactnormal _physmat _hitactor) (Collision|BreakHitResult (Variables|Default|GetHitResult)))
  (bind _hitresult (Variables|Default|GetHitResult))
  (bind _impact_location (Variables|VFX|GetImpact%20Location))
  (bind _counter (Variables|Default|GetCounter))
  ;[1 1 1] Event Spawn Explosion VFX
  (Variables|Default|SetCounter (+ _counter 1))
  (if (== (Math|Integer|%25%28Integer%29 _counter 2) 1)
    (bind _asfort_player_pawn_athena (Utilities|Casting|CastToFortPlayerPawnAthena _hitactor)
      (:then
        (Niagara|SpawnSystemAttached "/HighTower/Items/Tomato/Repulsors/FX/NS_HighTower_Tomato_PlayerImpact.NS_HighTower_Tomato_PlayerImpact" (Variables|/Script/Engine.Character|GetMesh _asfort_player_pawn_athena) (select (Utilities|Name|Equal(Name) (.ReturnValue (Components|SkinnedMesh|FindClosestBone (Variables|/Script/Engine.Character|GetMesh _asfort_player_pawn_athena) (Variables|VFX|GetImpact%20Location) 0.0 false)) "head") "neck_01" (.ReturnValue (Components|SkinnedMesh|FindClosestBone (Variables|/Script/Engine.Character|GetMesh _asfort_player_pawn_athena) (Variables|VFX|GetImpact%20Location) 0.0 false))) "0, 0, 0" "0, 0, 0" "KeepRelativeOffset" true))
      (:CastFailed
        (bind _returnvalue (Niagara|SpawnSystematLocation "/HighTower/Items/Tomato/Repulsors/FX/NS_HighTower_Tomato_Impact.NS_HighTower_Tomato_Impact" _impact_location (Math|Rotator|MakeRotfromZ _impactnormal)))
        (Variables|Default|SetBlastFX _returnvalue)
        (Effects|Components|ParticleSystem|SetVectorParameter (Variables|Default|GetBlastFX) "AccelerationDirection" (Math|Vector|Normalize _impactnormal))
        (bind _returnvalue_2 (Rendering|Decal|SpawnDecalatLocation "/HighTower/Items/Tomato/Repulsors/FX/MD_Hightower_Tomato_Crack.MD_Hightower_Tomato_Crack" "50.000000,200.000000,200.000000" _impact_location (Math|Rotator|MakeRotfromX _impactnormal) 3.0))
        (Variables|Default|SetCrackDecal _returnvalue_2)
        (Rendering|Components|Decal|SetFadeOut (Variables|Default|GetCrackDecal) 1.0 2.0 true)))))

(event EventOnStop
  (params (Hit :type Hit%20Result%20Structure%20%28by%20ref%29))
  (bind (_bblockinghit _binitialoverlap _time _distance _location _impactpoint) (Collision|BreakHitResult Hit))
  ;[1 1 1] Fix for bad VFX explosion location
  (Networking|SwitchHasAuthority
    (:Authority
      ;[1 1 1] Event on Stop
      (Variables|Default|SetHitResult Hit)
      (Variables|VFX|SetImpact%20Location _impactpoint)
      (Development|PrintString (Utilities|String|ToString(Vector) _location)))
    (:Remote
      _))
  (Utilities|FlowControl|Sequence
    (:then_0
      (Utilities|FlowControl|DoOnce false
        (:Completed
          (Variables|VFX|SetbHasExploded? true)
          (Delivery|DoExplosion) ; Explosion
        )))
    (:then_1
      (Networking|SwitchHasAuthority
        (:Authority
          (Utilities|Time|ClearandInvalidateTimerbyHandle (Variables|Default|GetLifespan%20Timer)))
        (:Remote
          (Utilities|FlowControl|Delay (Variables|VFX|GetFX%20Linger%20Time))
          (Actor|DestroyActor))))))

(event Custom|Timeout
  (Components|Movement|StopMovementImmediately (Variables|Movement|GetProjectileMovementComponent))
  (if (not (Variables|VFX|GetbHasExploded?))
    (Variables|VFX|SetImpact%20Location (Transformation|GetWorldLocation (Variables|VFX|GetTracerBody)))
    (Variables|VFX|SetbHasExploded? true)
    (Delivery|DoExplosion) ; Explosion
  ))

(event EventBeginPlay
  (bind _tracermesh (Variables|Default|GetTracerMesh))
  (Variables|VFX|SetFX%20Linger%20Time (Math|Random|RandomFloatinRange 0.5 1.0))
  (Networking|SwitchHasAuthority
    (:Authority
      (bind _returnvalue (Utilities|Time|SetTimerbyEvent (delegate Custom|Timeout) (Variables|Default|GetProjectile%20Max%20Lifetime)))
      (Variables|Default|SetLifespan%20Timer _returnvalue))
    (:Remote
      (bind _asb_high_tower_tomato_repulsors (Utilities|Casting|CastToB_HighTower_Tomato_Repulsors (Actor|GetOwner)))
      (bind _asfort_player_pawn (Utilities|Casting|CastToFortPlayerPawn (Variables|Actor|GetInstigator)))
      (Rendering|Material|SetVectorParameterValueonMaterials _tracermesh "BulletOrigin" (Transformation|GetSocketLocation (Customization|GetSkeletalMeshforPartType _asfort_player_pawn "Gameplay" :!class /Script/FortniteGame.FortPlayerPawn) (Utilities|Select :Index (Variables|/HighTower/Items/Tomato/Repulsors/B_HighTower_Tomato_Repulsors.B_HighTower_Tomato_Repulsors_C|GetLastHandFired _asb_high_tower_tomato_repulsors) :LEFT "FX_repulsor_l" :RIGHT "FX_repulsor_r")))
      (Rendering|Material|SetScalarParameterValueonMaterials _tracermesh "StartTime" (Utilities|Time|GetGameTimeinSeconds))
      (Rendering|Material|SetScalarParameterValueonMaterials _tracermesh "WPODuration" (Variables|Default|GetTracer%20Lifetime)))))

(event ProjectileMovement|EventOnHomingPassedTarget
  (Projectile|SetHomingStyle 0 0 "None" false :!class /Script/FortniteGame.FortProjectileMovementInterface))

(event ProjectileMovement|EventOnReachedLockTargetDistanceThreshold
  (Projectile|SetHomingStyle 0 0 "None" false :!class /Script/FortniteGame.FortProjectileMovementInterface))

(event OnEnterWaterEvent(WaterInteractionComponent)
  (params (WaterBody :type Fort%20Water%20Body%20Actor%20Object%20Reference) (WaterInteractionComponent :type Fort%20Water%20Interaction%20Component%20Object%20Reference) (bIsFirstBody :type Boolean))
  ;[1 1 1] On Enter Water, play Water Splash FX
  (Effects|Components|ParticleSystem|SpawnFXAtLocation "/Game/Weapons/Prototype/FringePlank/Particle/P_FringPlank_WaterImpact.P_FringPlank_WaterImpact" (Math|Vector|MakeVector (.x (Transformation|GetWorldLocation (Variables|Collision|GetCapsuleComponent))) (.y (Transformation|GetWorldLocation (Variables|Collision|GetCapsuleComponent))) (+ (.z (Transformation|GetWorldLocation (Variables|Collision|GetCapsuleComponent))) 25.0)))
  (Audio|PlaySoundatLocation 0 (Transformation|GetActorLocation)))

(fn OnRep_ImpactLocation
  (Components|Activation|Deactivate (Variables|VFX|GetTracerBody))
  (Networking|CanExecuteCosmeticEvents
    (:True
      (CallFunction|ExplosionVFX))))

(fn ConstructionScript)
