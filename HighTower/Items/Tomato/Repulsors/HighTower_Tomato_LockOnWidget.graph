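# Lock-on reticule widget for the HighTower Tomato Repulsors item, drawn as a cyan-tinted image centered on a canvas.
# The image uses the Ashton Indigo lock-on cursor material and starts collapsed, with no logic here to show it.
# The Update Rotation event writes its Rotation input to both the CenterSize and Intensity parameters of the image's dynamic material, despite the event's name.
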WidgetBlueprint /HighTower/Items/Tomato/Repulsors/HighTower_Tomato_LockOnWidget
graphs [EventGraph {nodes [
   n1 = Update Rotation!() -> (OutputDelegate, then, Rotation:double=~) {is=title K2Node_CustomEvent}
   > n2 = GetDynamicMaterial(self=Get Image_LockOnReticule() -> (Image_LockOnReticule:Image Object Reference))
   > SetScalarParameterValue(self=@n2, ParameterName=CenterSize, Value=@n1.Rotation)
   > SetScalarParameterValue(self=@n2, ParameterName=Intensity, Value=@n1.Rotation)
  ]}]
variables []
widgets {
 info {inheritedWidgetCount=0, parentClass=&UserWidget, rootWidgetClass=&CanvasPanel, widgetCount=2}
 widgets [
  CanvasPanel_29 {widgetClassPath=&CanvasPanel, widget=CanvasPanel_29}
  Image_LockOnReticule {
   widgetClassPath=&Image
   bIsVariable=true
   parent=CanvasPanel_29
   widget=Image_LockOnReticule
   properties {
    renderOpacity=2
    visibility=Collapsed
    brush {
     bIsDynamicallyLoaded=false
     drawAs=Image
     imageType=NoImage
     mirroring=NoMirror
     resourceName=None
     resourceObject=/Game/Athena/Items/Gameplay/BackPacks/Ashton/Indigo/FX/M_Ashton_Indigo_LockOnCursor
     tiling=NoTile
     imageSize=<32 32>
     margin=<ltrb 0 0 0 0>
     outlineSettings {
      bUseBrushTransparency=false
      roundingType=HalfHeightRadius
      width=0
      color {colorUseRule=UseColor_Specified, specifiedColor=<rgba 0 0 0 0>}
      cornerRadii=<0 0 0 0>
     }
     tintColor {colorUseRule=UseColor_Specified, specifiedColor=<rgba 0 0.9894 1 1>}
     uVRegion {bIsValid=false, max=<0 0>, min=<0 0>}
    }
    colorAndOpacity=<rgba 1 1 1 0.9>
   }
   slot {
    class=CanvasPanelSlot
    properties {layoutData {alignment=<0.5 0.5>, anchors {maximum=<0.5 0.5>, minimum=<0.5 0.5>}, offsets=<ltrb 0 0 250 250>}}
   }
  }
 ]
}
